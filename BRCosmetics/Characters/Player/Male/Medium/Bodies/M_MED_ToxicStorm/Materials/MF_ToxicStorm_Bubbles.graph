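# Material function for the ToxicStorm character body that blends an animated, glossy bubbling-liquid layer over incoming material attributes, using an FX-map channel mask and a static switch.
# Two flowmap-distorted panning bubble layers supply rims and blended normals.
# Liquid color comes from a constant or noise-warped texture.
# Emissive shading uses pulsing noisy Fresnel and sky-light backscatter, with evening and night dimming through Color_Time_of_Day.
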MaterialFunction /BRCosmetics/Characters/Player/Male/Medium/Bodies/M_MED_ToxicStorm/Materials/MF_ToxicStorm_Bubbles
graph {nodes [
  n1 = ChannelMaskParameter(None=FunctionInput() {inputName=InFXMap, sortPriority=1}) {group=BubbleLiquid, parameterName=BubbleLiquidMask, sortPriority=1}
  Floor(None=@n1.R)
  n2 = FunctionInput() {inputType=FunctionInput_MaterialAttributes}
  n3 = SetMaterialAttributes!(
   MaterialAttributes=@n2
   BaseColor=NamedRerouteUsage()
   EmissiveColor=NamedRerouteUsage()
   Metallic=ScalarParameter() {defaultValue=0.75, group=BubbleLiquid, parameterName=Metalness}
   Specular=ScalarParameter() {defaultValue=0.5, group=BubbleLiquid, parameterName=Specular}
   Roughness=ScalarParameter() {defaultValue=0.1, group=BubbleLiquid, parameterName=GlassRoughness}
   Normal=NamedRerouteUsage()
  ) {attributeSetTypes [
    #1
    #2
    #3
    #4
    #5
    #6
   ]}
  FunctionOutput(None=StaticSwitchParameter(True=BlendMaterialAttributes(A=@n2, B=@n3, Alpha=@n1.R), False=@n2) {group=BubbleLiquid, parameterName=UseBubbleLiquid, sortPriority=0})
  LinearInterpolate(Alpha=Saturate(None=NamedRerouteUsage())) {constA=0.4, constB=0.5}
  n4 = DDY(Value=TextureCoordinate() {coordinateIndex=1})
  n5 = Constant3Vector() {constant=<rgba 0 1 0 0>}
  Multiply(A=Sign(None=@n4), B=@n5)
  n6 = VectorParameter() {group=BubbleLiquid, parameterName=Potion-BubblesScaleAdjust, defaultValue=<rgba 1 1 0 1>}
  n7 = Multiply(
   A=Multiply(A=NamedRerouteUsage(), B=ComponentMask(None=@n6.RGB) {g=true, r=true})
   B=ScalarParameter() {defaultValue=1, group=BubbleLiquid, parameterName=OverallScale, sortPriority=0}
  )
  n8 = ScalarParameter() {defaultValue=4, group=BubbleLiquid, parameterName=Potion-BubblesTile1}
  n9 = Multiply(A=Time(), B=ScalarParameter() {defaultValue=1, group=BubbleLiquid, parameterName=Potion-BubblesSpeed})
  n10 = ScalarParameter() {defaultValue=4, group=BubbleLiquid, parameterName=Potion-FlowmapScale}
  n11 = Time()
  n12 = ScalarParameter() {defaultValue=0.2, group=BubbleLiquid, parameterName=Potion-FlowmapSpeed}
  n13 = TextureSample(UVs=Panner(Coordinate=Multiply(A=@n7, B=@n10), Time=Multiply(A=@n11, B=@n12)) {speedX=1}) {
   samplerSource=SSM_Wrap_WorldGroupSettings
   samplerType=SAMPLERTYPE_LinearColor
   texture=/Game/Global/Textures/Default/Flowmaps/T_TorusFlowmap
  }
  n14 = ComponentMask(None=Multiply(
    A=Subtract(A=@n13.RGB) {constB=0.5}
    B=VectorParameter() {group=BubbleLiquid, parameterName=Potion-FlowmapAmt, defaultValue=<rgba 1 1 0 1>}.RGB
   )) {g=true, r=true}
  n15 = Add(
   A=Panner(Coordinate=Multiply(A=@n7, B=@n8), Time=@n9) {speedX=-0.1, speedY=1}
   B=Multiply(A=ScalarParameter() {defaultValue=0.15, group=BubbleLiquid, parameterName=Potion-FlowmapStrength1}, B=@n14)
  )
  n16 = TextureSample(UVs=@n15) {samplerSource=SSM_Wrap_WorldGroupSettings, samplerType=SAMPLERTYPE_Masks, texture=/Game/Global/Textures/Default/PointSplatter/Bubbles_M}
  n17 = ScalarParameter() {defaultValue=2, group=BubbleLiquid, parameterName=Potion-BubblesTile2}
  n18 = Add(
   A=Panner(Coordinate=Multiply(A=@n7, B=@n17), Time=@n9) {speedX=0.1, speedY=1}
   B=Multiply(A=@n14, B=ScalarParameter() {defaultValue=0.1, group=BubbleLiquid, parameterName=Potion-FlowmapStrength2})
  )
  n19 = TextureSample(UVs=@n18) {samplerSource=SSM_Wrap_WorldGroupSettings, samplerType=SAMPLERTYPE_Masks, texture=&Bubbles_M}
  n20 = ComponentMask(None=Max(A=@n16.RGB, B=@n19.RGB)) {g=true}
  n21 = NamedRerouteUsage()
  n22 = Saturate(None=Add(A=Subtract(A=@n20), B=LinearInterpolate(Alpha=@n21) {constA=0.6}))
  n23 = OneMinus(None=@n22)
  n24 = Constant()
  n25 = Subtract(
   A=SphereMask(A=@n23, B=@n24) {attenuationRadius=1, hardnessPercent=1}
   B=SphereMask(A=@n23, B=@n24) {attenuationRadius=0.95, hardnessPercent=2}
  )
  NamedRerouteDeclaration(None=Multiply(A=Saturate(None=@n25), B=ScalarParameter() {defaultValue=12, group=BubbleLiquid, parameterName=BubblesIntensity})) {name=Bubbles, nodeColor=<rgba 0.0781 0 1 1>}
  NamedRerouteDeclaration(None=LinearInterpolate(
    A=TextureCoordinate()
    B=TextureCoordinate() {coordinateIndex=1}
    Alpha=ScalarParameter() {defaultValue=1, group=BubbleLiquid, parameterName=UseUV2, sliderMax=1}
   )) {name=UVs, nodeColor=<rgba 1 0.0435 0 1>}
  n26 = Multiply(A=NamedRerouteUsage(), B=ScalarParameter() {defaultValue=0.25, group=BubbleLiquid, parameterName=BaseColorDeformNoiseScale})
  n27 = ComponentMask(None=VectorParameter() {group=BubbleLiquid, parameterName=BaseColorDistortionSpeed, defaultValue=<rgba 0.05 0.1 0 1>}.RGB) {g=true, r=true}
  n28 = Subtract(A=TextureSample(UVs=Panner(Coordinate=@n26) {speedX=0.005, speedY=-0.05}) {samplerSource=SSM_Wrap_WorldGroupSettings, texture=/Game/Effects/Fort_Effects/Textures/Tile/T_GaussianNoise_01_M}.G) {constB=0.5}
  n29 = TextureSample(UVs=Subtract(A=Panner(Coordinate=@n26, Speed=@n27) {speedX=-0.025, speedY=0.1}, B=Multiply(A=@n28) {constB=0.25})) {samplerSource=SSM_Wrap_WorldGroupSettings, texture=&T_GaussianNoise_01_M}
  NamedRerouteDeclaration(None=@n29.G) {name=DistortionNoise, nodeColor=<rgba 1 0.375 0 1>}
  n30 = NamedRerouteUsage()
  n31 = ScalarParameter() {defaultValue=0.5, group=BubbleLiquid, parameterName=BubbleBurstMaskScale}
  n32 = TextureSample(UVs=Panner(Coordinate=Multiply(A=@n30, B=@n31)) {speedX=-0.025, speedY=0.1}) {
   samplerSource=SSM_Wrap_WorldGroupSettings
   samplerType=SAMPLERTYPE_LinearColor
   texture=/BRCosmetics/Weapons/FORT_Melee/Pickaxe_DyedDuelist/Textures/T_DyedDuelist_Noise
  }
  n33 = Power(Base=@n32.G) {constExponent=4}
  NamedRerouteDeclaration(None=Saturate(None=Multiply(A=@n33) {constB=7})) {name=BubbleBurstMask, nodeColor=<rgba 0.3828 1 0 1>}
  n34 = Fresnel() {baseReflectFraction=0, exponent=1}
  n35 = OneMinus(None=MaterialFunctionCall(Axis=Transform(None=@n5)) {materialFunction=&AxisAlignedFresnel}.Fresnel)
  n36 = ScalarParameter() {group=BubbleLiquid, parameterName=UseAxisAlignedFresnel}
  n37 = Sine(None=Multiply(A=Time()) {constB=0.2})
  n38 = Power(
   Base=NamedRerouteDeclaration(None=LinearInterpolate(A=@n34, B=@n35, Alpha=@n36)) {name=Fresnel, nodeColor=<rgba 1 0.1172 0 1>}
   Exp=Subtract(A=ScalarParameter() {defaultValue=1, group=BubbleLiquid, parameterName=FresnelPower}, B=Multiply(A=@n37) {constB=0.1})
  ) {constExponent=4}
  n39 = Abs(None=NamedRerouteDeclaration(None=@n37) {name=TimePulse, nodeColor=<rgba 0.1484 1 0 1>})
  n40 = Multiply(
   A=@n38
   B=Multiply(
    A=ScalarParameter() {defaultValue=1, group=BubbleLiquid, parameterName=FresnelIntensity}
    B=LinearInterpolate(Alpha=@n39) {constA=0.8, constB=1.2}
   )
  )
  n41 = NamedRerouteUsage()
  n42 = Saturate(None=Multiply(A=Multiply(A=@n41, B=@n41), B=ScalarParameter() {defaultValue=2, group=BubbleLiquid, parameterName=FresnelNoiseIntensity}) {constB=10})
  n43 = NamedRerouteDeclaration(None=Multiply(A=Add(A=@n40, B=@n42), B=@n40)) {name=NoisedFresnel, nodeColor=<rgba 0.4297 1 0 1>}
  n44 = Multiply(
   A=Multiply(A=NamedRerouteUsage(), B=ScalarParameter() {defaultValue=0.05, group=BubbleLiquid, parameterName=EmissiveBaseColorContribution}) {constB=0.1}
   B=OneMinus(None=@n43)
  )
  n45 = Multiply(A=NamedRerouteUsage(), B=OneMinus(None=@n38))
  n46 = Add(
   A=@n43
   B=Multiply(A=NamedRerouteUsage(), B=ScalarParameter() {defaultValue=0.25, group=BubbleLiquid, parameterName=ScatteringIntensity}) {constB=0.5}
  )
  n47 = Multiply(A=Add(A=Add(A=@n44, B=@n45), B=@n46), B=NamedRerouteUsage())
  n48 = Multiply(A=@n47, B=ScalarParameter() {defaultValue=0.75, group=BubbleLiquid, parameterName=Emissive_Night_Multiplier}) {constB=0.75}
  NamedRerouteDeclaration(None=MaterialFunctionCall(Color_Morning=@n47, Color_Day=@n47, Color_Evening=@n48, Color_Night=@n48) {materialFunction=&Color_Time_of_Day}.Result) {name=OutEmissive, nodeColor=<rgba 1 0 0.75 1>}
  n49 = PixelNormalWS()
  n50 = CameraVectorWS()
  n51 = Add(
   A=@n49
   B=Multiply(A=ScalarParameter() {defaultValue=1, group=BubbleLiquid, parameterName=ScatteringBacklightViewAngle, sortPriority=16}, B=@n50)
  )
  n52 = SkyAtmosphereLightDirection()
  n53 = Multiply(A=DotProduct(A=Normalize(VectorInput=@n51), B=Multiply(A=@n52) {constB=-1})) {constB=0.5}
  n54 = Power(
   Base=Saturate(None=Add(A=@n53) {constB=0.5})
   Exp=ScalarParameter() {defaultValue=8, group=BubbleLiquid, parameterName=ScatteringExpornent, sortPriority=17}
  )
  n55 = Saturate(None=DotProduct(A=@n49, B=@n50))
  n56 = NamedRerouteUsage()
  NamedRerouteDeclaration(None=Multiply(A=Multiply(A=@n54, B=@n55), B=OneMinus(None=@n56))) {name=Scattering, nodeColor=<rgba 0 1 0.1094 1>}
  n57 = TextureSample(UVs=@n15) {
   samplerSource=SSM_Wrap_WorldGroupSettings
   samplerType=SAMPLERTYPE_Normal
   texture=/BRCosmetics/Weapons/FORT_Melee/Pickaxe_GraffitiFry/Textures/FX/T_Bubbles_N
  }
  n58 = TextureSample(UVs=@n18) {
   samplerSource=SSM_Wrap_WorldGroupSettings
   samplerType=SAMPLERTYPE_Normal
   texture=&T_Bubbles_N
  }
  n59 = MaterialFunctionCall(BaseNormal=@n57.RGB, AdditionalNormal=@n58.RGB) {materialFunction=&BlendAngleCorrectedNormals}
  n60 = OneMinus(None=ScalarParameter() {defaultValue=1, parameterName=BubbleNormalIntensity})
  NamedRerouteDeclaration(None=LinearInterpolate(
    A=Constant3Vector() {constant=<rgba 0 0 1 0>}
    B=MaterialFunctionCall(Normal=@n59.Result, Flatness=@n60) {materialFunction=&FlattenNormal}.Result
    Alpha=@n22
   )) {name=OutNormals, nodeColor=<rgba 0 0.7891 1 1>}
  n61 = Subtract(A=@n29.G) {constB=0.5}
  n62 = ScalarParameter() {defaultValue=0.1, group=BubbleLiquid, parameterName=BaseColorDeformIntensity}
  n63 = TextureSampleParameter2D(UVs=Subtract(A=TextureCoordinate(), B=Multiply(A=@n61, B=@n62) {constB=0.1})) {group=BubbleLiquid, parameterName=LiquidColorTexture, texture=/Game/Characters/Fortress_Character/Textures/T_All_White}
  n64 = VectorParameter() {group=BubbleLiquid, parameterName=BaseColorMultiplier, sortPriority=2, defaultValue=<rgba 1 1 1 1>}
  n65 = VectorParameter() {group=BubbleLiquid, parameterName=BubbleLiquidColor, sortPriority=2, defaultValue=<rgba 0.0636 1 0.0168 0.3565>}
  NamedRerouteDeclaration(None=StaticSwitchParameter(True=Multiply(A=@n63.RGB, B=@n64.RGB), False=@n65.RGB) {group=BubbleLiquid, parameterName=UseBaseColorTexture}) {name=PotionColor, nodeColor=<rgba 1 0.0703 0 1>}
  n66 = ScalarParameter() {defaultValue=0.1, group=BubbleLiquid, parameterName=BaseColorMin}
  n67 = ScalarParameter() {group=BubbleLiquid, parameterName=BaseColorMax}
  n68 = NamedRerouteUsage()
  n69 = NamedRerouteUsage()
  n70 = Saturate(None=Add(A=Add(A=@n68, B=@n69), B=NamedRerouteUsage()))
  NamedRerouteDeclaration(None=Multiply(A=LinearInterpolate(A=@n66, B=@n67, Alpha=@n70) {constA=0.15}, B=NamedRerouteUsage())) {name=OutBaseColor, nodeColor=<rgba 0 0.25 1 1>}
  n71 = Multiply(A=NamedRerouteUsage(), B=ScalarParameter() {defaultValue=1, group=BubbleLiquid, parameterName=LiquidNoiseScale})
  n72 = TextureSample(UVs=Panner(Coordinate=@n71) {speedX=0.005, speedY=-0.05}) {
   samplerSource=SSM_Wrap_WorldGroupSettings
   samplerType=SAMPLERTYPE_LinearColor
   texture=&T_DyedDuelist_Noise
  }
  n73 = Subtract(A=@n72.G) {constB=0.5}
  n74 = TextureSample(UVs=Add(A=Panner(Coordinate=@n71) {speedX=-0.025, speedY=0.15}, B=Multiply(A=@n73) {constB=0.25})) {
   samplerSource=SSM_Wrap_WorldGroupSettings
   samplerType=SAMPLERTYPE_LinearColor
   texture=&T_DyedDuelist_Noise
  }
  NamedRerouteDeclaration(None=@n74.G) {name=WarpedNoise, nodeColor=<rgba 1 0 0.5859 1>}
  NamedRerouteUsage()
  NamedRerouteUsage()
  OneMinus(None=ConstantBiasScale(None=Normalize(VectorInput=@n4)))
  Saturate(None=@n46)
 ]}
