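# Physics asset for the front cloth of the male medium ToxicStorm cosmetic body, giving its hem bone chain capsule bodies and joint limits.
# Two dyn_hem_f bones chain from base_hem_f_00, which is locked to spine_04, with limited swing and zero twist limit.
# Capsules on pelvis, spine, thighs and pouch form the surrounding body shape.
# Thighs and pouch lock to the pelvis, while spine joints rotate freely.
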PhysicsAsset /BRCosmetics/Characters/Player/Male/Medium/Bodies/M_MED_ToxicStorm/Meshes/M_MED_ToxicStorm_FrontCloth_Physics
bodies {
 base_hem_b_00 [{shapeType=Capsule, length=0.505, radius=0.505, shapeName=base_hem_b_00_capsule, rotation=<rot 0 -90 0>}]
 base_hem_f_00 [{shapeType=Capsule, length=0.505, radius=0.505, shapeName=base_hem_f_00_capsule, rotation=<rot 0 -90 0>}]
 base_pouch_00 [{shapeType=Capsule, length=3.5, radius=6.5, shapeName=base_pouch_00_capsule, center=<-0.6419 -0.5514 0>}]
 dyn_hem_f_00 [{shapeType=Capsule, length=19.4509, radius=1.5841, shapeName=dyn_hem_f_00_capsule, center=<0.3294 0 0>}]
 dyn_hem_f_01 [
  {shapeType=Capsule, length=11.5, radius=2.5, shapeName=dyn_hem_f_01_capsule, center=<8.5 0.1613 0>}
  {shapeType=Capsule, length=15.5, radius=2.5, shapeName=dyn_hem_f_01_capsule, center=<4.6263 0.5331 0>}
  {shapeType=Capsule, length=18, radius=2.5, shapeName=dyn_hem_f_01_capsule, center=<-0.1395 0.3987 0>}
 ]
 pelvis [{shapeType=Capsule, length=7.6, radius=13.5, shapeName=pelvis_capsule}]
 spine_01 [
  {shapeType=Capsule, length=11, radius=13, shapeName=spine_01_capsule}
  {shapeType=Capsule, length=3.5, radius=7.5, shapeName=spine_01_capsule, center=<1.2042 6.8334 12.607>, rotation=<rot 60 89.9996 90>}
 ]
 spine_02 [{shapeType=Capsule, length=17, radius=11.5, shapeName=spine_02_capsule}]
 spine_03 [{shapeType=Capsule, length=10, radius=12, shapeName=spine_03_capsule, rotation=<rot 0 -0.0001 0>}]
 spine_04 [{shapeType=Capsule, length=11, radius=12, shapeName=spine_04_capsule}]
 spine_05 [{shapeType=Capsule, length=14.5, radius=12, shapeName=spine_05_capsule}]
 thigh_l [{shapeType=Capsule, length=24, radius=8, shapeName=thigh_l_capsule, center=<-17 0 0>, rotation=<rot 90 0 0>}]
 thigh_r [{shapeType=Capsule, length=24, radius=8, shapeName=thigh_l_capsule, center=<17 0 0>, rotation=<rot 90 0 0>}]
}
constraints [
 {bone1Name=dyn_hem_f_00, bone2Name=base_hem_f_00, swing1LimitDegrees=20, swing2LimitDegrees=20, twistLimitDegrees=0}
 {bone1Name=dyn_hem_f_01, bone2Name=dyn_hem_f_00, swing1LimitDegrees=30, swing2LimitDegrees=5, twistLimitDegrees=0}
 {
  bone1Name=base_hem_f_00
  bone2Name=spine_04
  swing1LimitDegrees=45
  swing1Motion=Locked
  swing2LimitDegrees=45
  swing2Motion=Locked
  twistLimitDegrees=45
  twistMotion=Locked
 }
 {
  bone1Name=base_hem_b_00
  bone2Name=spine_04
  swing1LimitDegrees=45
  swing1Motion=Free
  swing2LimitDegrees=45
  swing2Motion=Free
  twistLimitDegrees=45
  twistMotion=Free
 }
 {
  bone1Name=spine_01
  bone2Name=pelvis
  swing1LimitDegrees=45
  swing1Motion=Free
  swing2LimitDegrees=45
  swing2Motion=Free
  twistLimitDegrees=45
  twistMotion=Free
 }
 {
  bone1Name=spine_02
  bone2Name=spine_01
  swing1LimitDegrees=45
  swing1Motion=Free
  swing2LimitDegrees=45
  swing2Motion=Free
  twistLimitDegrees=45
  twistMotion=Free
 }
 {
  bone1Name=spine_03
  bone2Name=spine_02
  swing1LimitDegrees=45
  swing1Motion=Free
  swing2LimitDegrees=45
  swing2Motion=Free
  twistLimitDegrees=45
  twistMotion=Free
 }
 {
  bone1Name=spine_04
  bone2Name=spine_03
  swing1LimitDegrees=45
  swing1Motion=Free
  swing2LimitDegrees=45
  swing2Motion=Free
  twistLimitDegrees=45
  twistMotion=Free
 }
 {
  bone1Name=spine_05
  bone2Name=spine_04
  swing1LimitDegrees=45
  swing1Motion=Free
  swing2LimitDegrees=45
  swing2Motion=Free
  twistLimitDegrees=45
  twistMotion=Free
 }
 {
  bone1Name=base_pouch_00
  bone2Name=pelvis
  swing1LimitDegrees=45
  swing1Motion=Locked
  swing2LimitDegrees=45
  swing2Motion=Locked
  twistLimitDegrees=45
  twistMotion=Locked
 }
 {
  bone1Name=thigh_l
  bone2Name=pelvis
  swing1LimitDegrees=45
  swing1Motion=Locked
  swing2LimitDegrees=45
  swing2Motion=Locked
  twistLimitDegrees=45
  twistMotion=Locked
 }
 {
  bone1Name=thigh_r
  bone2Name=pelvis
  swing1LimitDegrees=45
  swing1Motion=Locked
  swing2LimitDegrees=45
  swing2Motion=Locked
  twistLimitDegrees=45
  twistMotion=Locked
 }
 {
  bone1Name=base_pouch_00
  bone2Name=thigh_r
  swing1LimitDegrees=45
  swing1Motion=Free
  swing2LimitDegrees=45
  swing2Motion=Free
  twistLimitDegrees=45
  twistMotion=Free
 }
 {
  bone1Name=thigh_l
  bone2Name=thigh_r
  swing1LimitDegrees=45
  swing1Motion=Free
  swing2LimitDegrees=45
  swing2Motion=Free
  twistLimitDegrees=45
  twistMotion=Free
 }
]
