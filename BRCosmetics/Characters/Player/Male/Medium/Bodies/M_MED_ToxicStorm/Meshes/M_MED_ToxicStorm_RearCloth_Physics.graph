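# Physics asset for the rear cloth of the male medium ToxicStorm body cosmetic, modelling the back hem as jointed capsule bodies.
# Pelvis and spine capsules form a freely rotating chain up to a hem base, whose left and right anchors are angularly locked.
# Each anchor carries a three-segment dynamic hem chain with limited swing and zero-degree twist limits.
# A cross constraint links the two middle segments, coupling the left and right sides.
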PhysicsAsset /BRCosmetics/Characters/Player/Male/Medium/Bodies/M_MED_ToxicStorm/Meshes/M_MED_ToxicStorm_RearCloth_Physics
bodies {
 base_hem_b_00 [{shapeType=Capsule, length=0.5, radius=2.005, shapeName=base_hem_b_00_capsule, rotation=<rot 0 -90 0>}]
 base_hem_bl_00 [{shapeType=Capsule, length=0.505, radius=0.505, shapeName=base_hem_bl_00_capsule, rotation=<rot 0 -90 0>}]
 base_hem_br_00 [{shapeType=Capsule, length=0.505, radius=0.505, shapeName=base_hem_br_00_capsule, rotation=<rot 0 -90 0>}]
 dyn_hem_bl_00 [{shapeType=Capsule, length=11.5, radius=2.5, shapeName=dyn_hem_bl_00_capsule, center=<7.5464 -1 0>}]
 dyn_hem_bl_01 [{shapeType=Capsule, length=10, radius=3.5, shapeName=dyn_hem_bl_01_capsule, center=<1 -0.6 0>}]
 dyn_hem_bl_02 [{shapeType=Capsule, length=11, radius=3.5, shapeName=dyn_hem_bl_02_capsule, center=<0.5001 -0.6 0>}]
 dyn_hem_br_00 [{shapeType=Capsule, length=11.5, radius=2.5, shapeName=dyn_hem_br_00_capsule, center=<7.6546 -1 0>}]
 dyn_hem_br_01 [{shapeType=Capsule, length=10, radius=3.4036, shapeName=dyn_hem_br_01_capsule, center=<1 -0.6 0>}]
 dyn_hem_br_02 [{shapeType=Capsule, length=11, radius=3.5, shapeName=dyn_hem_br_02_capsule, center=<0.9788 -0.6 0>}]
 pelvis [{shapeType=Capsule, length=7.5, radius=13.5, shapeName=pelvis_capsule}]
 spine_01 [{shapeType=Capsule, length=14.5, radius=11, shapeName=spine_01_capsule, center=<0 0.1251 0>}]
 spine_02 [{shapeType=Capsule, length=18.5, radius=10, shapeName=spine_02_capsule, center=<0 -0.2136 0>}]
 spine_03 [{shapeType=Capsule, length=12, radius=10, shapeName=spine_03_capsule, center=<0 0.1047 0>}]
 spine_04 [{shapeType=Capsule, length=11, radius=12, shapeName=spine_04_capsule, center=<0 0.4843 0>}]
 spine_05 [{shapeType=Capsule, length=14.5, radius=12, shapeName=spine_05_capsule, center=<-4.8083 1.1251 0>}]
}
constraints [
 {
  bone1Name=spine_01
  bone2Name=pelvis
  swing1LimitDegrees=45
  swing1Motion=Free
  swing2LimitDegrees=45
  swing2Motion=Free
  twistLimitDegrees=45
  twistMotion=Free
 }
 {
  bone1Name=spine_02
  bone2Name=spine_01
  swing1LimitDegrees=45
  swing1Motion=Free
  swing2LimitDegrees=45
  swing2Motion=Free
  twistLimitDegrees=45
  twistMotion=Free
 }
 {
  bone1Name=spine_03
  bone2Name=spine_02
  swing1LimitDegrees=45
  swing1Motion=Free
  swing2LimitDegrees=45
  swing2Motion=Free
  twistLimitDegrees=45
  twistMotion=Free
 }
 {
  bone1Name=spine_04
  bone2Name=spine_03
  swing1LimitDegrees=45
  swing1Motion=Free
  swing2LimitDegrees=45
  swing2Motion=Free
  twistLimitDegrees=45
  twistMotion=Free
 }
 {
  bone1Name=spine_05
  bone2Name=spine_04
  swing1LimitDegrees=45
  swing1Motion=Free
  swing2LimitDegrees=45
  swing2Motion=Free
  twistLimitDegrees=45
  twistMotion=Free
 }
 {
  bone1Name=base_hem_b_00
  bone2Name=spine_05
  swing1LimitDegrees=45
  swing1Motion=Free
  swing2LimitDegrees=45
  swing2Motion=Free
  twistLimitDegrees=45
  twistMotion=Free
 }
 {
  bone1Name=base_hem_bl_00
  bone2Name=base_hem_b_00
  swing1LimitDegrees=45
  swing1Motion=Locked
  swing2LimitDegrees=45
  swing2Motion=Locked
  twistLimitDegrees=45
  twistMotion=Locked
 }
 {bone1Name=dyn_hem_bl_00, bone2Name=base_hem_bl_00, swing1LimitDegrees=35, swing2LimitDegrees=10, twistLimitDegrees=0}
 {bone1Name=dyn_hem_bl_01, bone2Name=dyn_hem_bl_00, swing1LimitDegrees=45, swing2LimitDegrees=15, twistLimitDegrees=0}
 {
  bone1Name=base_hem_br_00
  bone2Name=base_hem_b_00
  swing1LimitDegrees=45
  swing1Motion=Locked
  swing2LimitDegrees=45
  swing2Motion=Locked
  twistLimitDegrees=45
  twistMotion=Locked
 }
 {bone1Name=dyn_hem_br_00, bone2Name=base_hem_br_00, swing1LimitDegrees=30, swing2LimitDegrees=10, twistLimitDegrees=0}
 {bone1Name=dyn_hem_br_01, bone2Name=dyn_hem_br_00, swing1LimitDegrees=40, swing2LimitDegrees=15, twistLimitDegrees=0}
 {bone1Name=dyn_hem_br_01, bone2Name=dyn_hem_bl_01, swing1LimitDegrees=45, swing2LimitDegrees=45, twistLimitDegrees=45}
 {bone1Name=dyn_hem_br_02, bone2Name=dyn_hem_br_01, swing1LimitDegrees=50, swing2LimitDegrees=0, twistLimitDegrees=0}
 {bone1Name=dyn_hem_bl_02, bone2Name=dyn_hem_bl_01, swing1LimitDegrees=50, swing2LimitDegrees=0, twistLimitDegrees=0}
]
