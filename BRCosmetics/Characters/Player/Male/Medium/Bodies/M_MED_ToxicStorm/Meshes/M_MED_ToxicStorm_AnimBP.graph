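# Anim Blueprint for the M_MED_ToxicStorm cosmetic body, adding secondary cloth motion to the copied parent pose.
# A Control Rig driven by crouch, glide, movement and aim feeds a hem bone offset, then front and rear RigidBody simulations.
# Rear cloth is disabled for qualifying backblings detected on customization load.
# Each update, a deprecated helper selects per-state rear sim-space settings; its gravity result is stored but unused by the AnimGraph.
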AnimBlueprint /BRCosmetics/Characters/Player/Male/Medium/Bodies/M_MED_ToxicStorm/Meshes/M_MED_ToxicStorm_AnimBP
target_skeleton=&/M_MED_ToxicStorm_Skeleton
graphs [
 AnimGraph {nodes [
   n1 = Control Rig!(
    Source=Copy Pose From Mesh() {bUseAttachedParent=true}
    Alpha
    iscrouching:double=Get CrouchingAlpha() -> (CrouchingAlpha:double)
    isgliding:double=Get GlidingAlpha() -> (GlidingAlpha:double)
    moving2Dalpha:double=Get moving2Dalpha() -> (moving2Dalpha:double)
    isAim:bool=Get bIsTargeting()
   ) -> (Pose) {
    alphaBoolBlend {blendOption=Linear}
    controlRigAssetReference {blueprintRigClass=&/M_MED_ToxicStorm_CtrlRig.M_MED_ToxicStorm_CtrlRig_C}
   }
   n2 = "Transform (Modify) Bone - Bone: base_hem_b_00"!(ComponentPose=Local To Component(LocalPose=@n1), bAlphaBoolEnabled:bool=True, Translation="0,-2,0", Rotation="0,0,0", Scale="1,1,1") -> (Pose) {
    alphaInputType=Bool
    translationMode=BMM_Additive
    alphaBoolBlend {blendInTime=0.2, blendOption=QuadraticInOut, blendOutTime=0.8}
    boneToModify {boneName=base_hem_b_00}
    translation {y=-2}
   }
   n3 = RigidBody!(ComponentPose=@n2, Alpha, OverrideWorldGravity:Vector="0,0,0", ExternalForce) -> (Pose) {
    bOverrideWorldGravity=true
    overridePhysicsAsset=&/M_MED_ToxicStorm_FrontCloth_Physics
    simulationSpace=BaseBoneSpace
    alphaBoolBlend {blendOption=Linear}
    baseBoneRef {boneName=spine_05}
    simSpaceSettings {maxAngularAcceleration=40, maxAngularVelocity=8, maxLinearAcceleration=3500, maxLinearVelocity=700, velocityScaleZ=0.4, worldAlpha=1}
   }
   n4 = RigidBody!(
    ComponentPose=@n3
    bAlphaBoolEnabled:bool=NOT Boolean(A=Get IsWearingBackbling() -> (IsWearingBackbling:bool))
    OverrideWorldGravity="0,0,0"
    ExternalForce
    SimSpaceSettings:Sim Space Settings Structure=Get Sim Space Settings() -> (Sim Space Settings:Sim Space Settings Structure)
   ) -> (Pose) {
    alphaInputType=Bool
    bOverrideWorldGravity=true
    overridePhysicsAsset=&/M_MED_ToxicStorm_RearCloth_Physics
    simulationSpace=BaseBoneSpace
    alphaBoolBlend {blendOption=Linear}
    baseBoneRef {boneName=spine_05}
    simSpaceSettings {
     maxAngularAcceleration=50
     maxAngularVelocity=10
     maxLinearAcceleration=3500
     maxLinearVelocity=700
     worldAlpha=1
     externalLinearDragV=<0.2 0.2 0.2>
    }
   }
   Output Pose(Result=Component To Local(ComponentPose=@n4)) {layerGroup=None}
  ]}
 EventGraph {nodes [
   Event BlueprintUpdateAnimation()
   > n1 = Gravity Override State Function DEPRECATED(
    CharacterPartAnimInstance=Self-Reference()
    GameTimeInSeconds=GetGameTimeInSeconds()
    GravityOverrideParams=Get Gravity Override Params()
    RBAN Params=Get RBAN Params()
   )
   > Set GravityOverride(GravityOverride=@n1)
   > Set Sim Space Settings(Sim Space Settings:Sim Space Settings Structure=Break RBANSetting Struct!(RBANSettingStruct:"RBANSetting Struct Structure (by ref)"=@n1.RBAN SettingsOut) -> (
     ComponentLinearAccScale_28_#1:Vector=~
     ComponentLinearVelScale_31_#2:Vector=~
     SimSpaceSettings_34_#3:Sim Space Settings Structure
    ) {is=title K2Node_BreakStruct}.SimSpaceSettings_34_#3) -> (Output_Get:Sim Space Settings Structure)
   > n2 = Sequence()
   n3 = Branch(execute=@n2.then_1, Condition=OR Boolean(A=Get bIsCrouchSprinting(), B=Get bIsMoving2D()))
   > Set moving2Dalpha(moving2Dalpha:double=1) -> (Output_Get:double)
   Set moving2Dalpha(execute=@n3.else, moving2Dalpha=0)
   n4 = Branch(execute=@n2.then_0, Condition=Get IsWearingBackbling())
   > Set IsMoving2D_var(IsMoving2D_var:bool=false) -> (Output_Get:bool)
   n5 = Branch(execute=@n4.else, Condition=Get bIsMoving2D())
   > n6 = Branch(Condition=Get bIsCrouchSprinting())
   > Set IsMoving2D_var(IsMoving2D_var=false)
   Set IsMoving2D_var(execute=@n6.else, IsMoving2D_var=true)
   Set IsMoving2D_var(execute=@n5.else, IsMoving2D_var=false)
   Event OnBeginCrouching()
   > Set CrouchingAlpha(CrouchingAlpha:double=1) -> (Output_Get:double)
   Event OnEndCrouching()
   > Set CrouchingAlpha(CrouchingAlpha=0)
   Event OnEndSkydiving()
   > Set GlidingAlpha(GlidingAlpha:double=0) -> (Output_Get:double)
   Event OnBeginSkydiving()
   > Set GlidingAlpha(GlidingAlpha=1)
   Event OnCharacterCustomizationLoaded()
   n7 = CheckForBackbling()
   > Set IsWearingBackbling(IsWearingBackbling:bool=OR Boolean!(A=@n7, B=@n7.bIsBackblingACape, C:bool=@n7.bIsBackblingLong) -> (ReturnValue)) -> (Output_Get:bool)
  ]}
]
variables [
 CrouchingAlpha {type=number}
 GlidingAlpha {type=number}
 Gravity Override Params {type=GravityOverrideParamsStruct}
 GravityOverride {type=Vector}
 IsMoving2D_var {type=boolean}
 IsWearingBackbling {type=boolean}
 moving2Dalpha {type=number}
 RBAN Params {
  type=RBANParamsStruct
  default {
   dBNO {simSpaceSettings {
     maxAngularAcceleration=50
     maxAngularVelocity=10
     maxLinearAcceleration=3500
     maxLinearVelocity=700
     velocityScaleZ=1
     externalLinearDragV=<0.2 0.2 0>
    }}
   dBNO_Carried {simSpaceSettings {
     maxAngularAcceleration=50
     maxAngularVelocity=10
     maxLinearAcceleration=3500
     maxLinearVelocity=700
     velocityScaleZ=1
     externalLinearDragV=<0.2 0.2 0>
    }}
   emoteOrMelee {simSpaceSettings {
     maxAngularAcceleration=50
     maxAngularVelocity=10
     maxLinearAcceleration=3500
     maxLinearVelocity=700
     velocityScaleZ=1
     worldAlpha=0.4
     externalLinearDragV=<0.2 0.2 0>
    }}
   front_End {simSpaceSettings {
     maxAngularAcceleration=50
     maxAngularVelocity=10
     maxLinearAcceleration=3500
     maxLinearVelocity=700
     velocityScaleZ=1
     worldAlpha=0.2
     externalLinearDragV=<0.2 0.2 0>
    }}
   onGround_Crouch_Moving {simSpaceSettings {
     maxAngularAcceleration=50
     maxAngularVelocity=10
     maxLinearAcceleration=3500
     maxLinearVelocity=700
     velocityScaleZ=1
     externalLinearDragV=<0.2 0.2 0.2>
    }}
   onGround_Falling {simSpaceSettings {
     maxAngularAcceleration=50
     maxAngularVelocity=10
     maxLinearAcceleration=3500
     maxLinearVelocity=700
     velocityScaleZ=1
     externalLinearDragV=<0.2 0.2 0>
    }}
   onGround_Moving {simSpaceSettings {
     maxAngularAcceleration=50
     maxAngularVelocity=10
     maxLinearAcceleration=3500
     maxLinearVelocity=700
     velocityScaleZ=1
     externalLinearDragV=<0.2 0.2 0>
    }}
   onGround_Standing {
    componentLinearAccScale=<0 0 0>
    componentLinearVelScale=<0 0 0>
    simSpaceSettings {
     maxAngularAcceleration=50
     maxAngularVelocity=10
     maxLinearAcceleration=3500
     maxLinearVelocity=700
     velocityScaleZ=1
     worldAlpha=0.8
     externalLinearDragV=<0.2 0.2 0>
    }
   }
   rail_Grinding {simSpaceSettings {
     maxAngularAcceleration=50
     maxAngularVelocity=10
     maxLinearAcceleration=3500
     maxLinearVelocity=700
     velocityScaleZ=1
     worldAlpha=1
     externalLinearDragV=<0.2 0.2 0>
    }}
   riding_Moving {simSpaceSettings {
     maxAngularAcceleration=50
     maxAngularVelocity=10
     maxLinearAcceleration=3500
     maxLinearVelocity=700
     velocityScaleZ=1
     externalLinearDragV=<0.2 0.2 0>
    }}
   riding_Moving_Targeting {simSpaceSettings {
     maxAngularAcceleration=50
     maxAngularVelocity=10
     maxLinearAcceleration=3500
     maxLinearVelocity=700
     velocityScaleZ=1
     externalLinearDragV=<0.2 0.2 0>
    }}
   riding_Standing {simSpaceSettings {
     maxAngularAcceleration=50
     maxAngularVelocity=10
     maxLinearAcceleration=3500
     maxLinearVelocity=700
     velocityScaleZ=1
     externalLinearDragV=<0.2 0.2 0>
    }}
   skydive_DivingDown {simSpaceSettings {
     maxAngularAcceleration=50
     maxAngularVelocity=10
     maxLinearAcceleration=3500
     maxLinearVelocity=700
     velocityScaleZ=1
     worldAlpha=0.5
     externalLinearDragV=<0.2 0.2 0>
    }}
   skydive_FloatingDown {
    componentLinearAccScale=<0 0.5 0>
    componentLinearVelScale=<0 0.5 0>
    simSpaceSettings {
     maxAngularAcceleration=50
     maxAngularVelocity=10
     maxLinearAcceleration=3500
     maxLinearVelocity=700
     velocityScaleZ=1
     worldAlpha=0.5
     externalLinearDragV=<0.2 0.2 0>
    }
   }
   skydive_GoingUp {simSpaceSettings {
     maxAngularAcceleration=50
     maxAngularVelocity=10
     maxLinearAcceleration=3500
     maxLinearVelocity=700
     velocityScaleZ=1
     externalLinearDragV=<0.2 0.2 0>
    }}
   skydive_Parachuting {
    componentLinearAccScale=<0 0.1 0.3>
    componentLinearVelScale=<0 0.1 0.3>
    simSpaceSettings {
     maxAngularAcceleration=50
     maxAngularVelocity=10
     maxLinearAcceleration=3500
     maxLinearVelocity=700
     velocityScaleZ=1
     worldAlpha=0.5
     externalLinearDragV=<0.2 0.2 0>
    }
   }
   skydive_Surfing {simSpaceSettings {
     maxAngularAcceleration=50
     maxAngularVelocity=10
     maxLinearAcceleration=3500
     maxLinearVelocity=700
     velocityScaleZ=1
     externalLinearDragV=<0.2 0.2 0>
    }}
   swimming {simSpaceSettings {
     maxAngularAcceleration=50
     maxAngularVelocity=10
     maxLinearAcceleration=3500
     maxLinearVelocity=700
     velocityScaleZ=1
     externalLinearDragV=<0.2 0.2 0>
    }}
   tactical_Sprint {simSpaceSettings {
     maxAngularAcceleration=50
     maxAngularVelocity=10
     maxLinearAcceleration=3500
     maxLinearVelocity=700
     velocityScaleZ=1
     externalLinearDragV=<0.2 0.2 0>
    }}
   ziplining {simSpaceSettings {
     maxAngularAcceleration=50
     maxAngularVelocity=10
     maxLinearAcceleration=3500
     maxLinearVelocity=700
     velocityScaleZ=1
     worldAlpha=0.7
     externalLinearDragV=<0.2 0.2 0>
    }}
  }
 }
 Sim Space Settings {
  type=SimSpaceSettings
  default {
   dampingAlpha=1
   maxAngularAcceleration=10000
   maxAngularVelocity=10000
   maxLinearAcceleration=10000
   maxLinearVelocity=10000
   velocityScaleZ=1
   worldAlpha=0
   externalLinearDragV=<0 0 0>
  }
 }
]
