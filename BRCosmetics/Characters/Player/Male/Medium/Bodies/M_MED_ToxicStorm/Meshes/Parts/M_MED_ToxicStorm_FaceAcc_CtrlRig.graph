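# Control Rig for the male medium ToxicStorm face accessory part that corrects hood bones as the head turns.
# A sequence runs the shared DeformRig function, then DeltaRotation, which remaps head-to-neck_01 Euler X and Z into weights.
# These weights drive chained Modify Transforms offsets on the hood bones, with negative X shaped by a curve.
# A second DeltaRotation pass uses another Z range for an extra offset; a Multiply node is left unused.
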ControlRigBlueprint /BRCosmetics/Characters/Player/Male/Medium/Bodies/M_MED_ToxicStorm/Meshes/Parts/M_MED_ToxicStorm_FaceAcc_CtrlRig
graphs [
 New Function_ContainedGraph {nodes [
   n1 = RigVMFunctionEntryNode()
   RigVMFunction_MathQuaternionToEuler = &RigVMFunction_MathQuaternionToEuler(
    RotationOrder=ZYX
    Value=&RigUnit_GetRelativeTransformForItem(
     Child=(Type=Bone,Name="head")
     Parent=(Type=Bone,Name="neck_01")
     bChildInitial=False
     bParentInitial=False
     Child.Name=@n1.Child
     Parent.Name=@n1.Parent
    ).RelativeTransform.Rotation
   )
   Remap = &RigVMFunction_MathDoubleRemap(
    TargetMaximum=1
    TargetMinimum=0
    SourceMaximum=@n1.Source Max
    SourceMinimum=@n1.Source Min
    Value=@RigVMFunction_MathQuaternionToEuler.Result.Z
   )
   > RigVMFunctionReturnNode(
    Delta_z=@Remap.Result
    Delta_X_Negative=&RigVMFunction_MathDoubleRemap(SourceMaximum=-60, SourceMinimum=-1, TargetMaximum=1, TargetMinimum=0, Value=@RigVMFunction_MathQuaternionToEuler.Result.X).Result
    Delta_X_Positive=&RigVMFunction_MathDoubleRemap(SourceMaximum=60, SourceMinimum=0, TargetMaximum=1, TargetMinimum=0, Value=@RigVMFunction_MathQuaternionToEuler.Result.X).Result
   )
  ]}
 RigVMFunctionLibrary {nodes [DeltaRotation = RigVMCollapseNode(ExecuteContext=())]}
 RigVMModel {nodes [
   RigUnit_BeginExecution = &RigUnit_BeginExecution()
   RigVMFunction_Sequence = &RigVMFunction_Sequence(ExecuteContext=@RigUnit_BeginExecution.ExecutePin)
   FN_DeformRig_FNC = RigVMFunctionReferenceNode(
    DeformRig_ Curve="(deformRig_l_upperarm_up_alpha_72_#1=False,deformRig_r_upperarm_up_alpha_73_#2=False,deformRig_l_upperarm_fr_alpha_74_#3=False,deformRig_r_upperarm_fr_alpha_75_CB51B8 ...(truncated)"
    DeformRig_Bool="(upperarm_1_#4=True,elbow_3_#5=True,wrist_7_#6=True,groin_8_#7=True,glute_17_#8=True,knee_9_40C408324CA261 ...(truncated)"
    DeformRig_Debug=(upperarm_debug_18_#4=False,wrist_debug_20_#6=False,lat_debug_26_#9=False,groinGlute_debug_27_#7=False)
    DeformRig_Value="(upperarm_up_18_#10=(Rotation=(X=0,Y=0,Z=0,W=1),Translation=(X=0,Y=0,Z=8),Scale3D=(X=0.7,Y=1,Z=1)),upperarm_fwd_fr_19_#11=(Rotation=(X=0,Y=0,Z=0,W=1),Translation=(X=0,Y=10,Z=0),Scale3D=(X=0.5,Y=1,Z= ...(truncated)"
    ExecuteContext=@RigVMFunction_Sequence.A
   )
   DeltaRotation = RigVMFunctionReferenceNode(Child=Head, Parent=neck_01, Source Max=-60, Source Min=-20, ExecuteContext=@RigVMFunction_Sequence.B)
   RigVMFunction_AnimEvalRichCurve = &RigVMFunction_AnimEvalRichCurve(
    Curve="(EditorCurveData=(Keys=((InterpMode=RCIM_Linear,TangentMode=RCTM_Auto,TangentWeightMode=RCTWM_WeightedNone,Time=0,Value=0,ArriveTangent=0,ArriveTangentWeight=0,LeaveTangent=0,LeaveTangentWeight=0),(InterpMode=RCIM_Linear,TangentMode=RCTM_Auto,TangentWeight ...(truncated)"
    SourceMinimum=0
    TargetMinimum=0
    Value=@DeltaRotation.Delta_X_Negative
   )
   > &RigUnit_ModifyTransforms(
    ItemToModify="((Item=(Type=Bone,Name=\"base_hood_l1\"),Transform=(Rotation=(X=0,Y=0,Z=0,W=-1),Translation=(X=9.019,Y=0,Z=0),Scale3D=(X=1,Y=1,Z=0.5))),(Item=(Type=Bone,Name=\"base_hood_l\"),Transform=(Rotation=(X=0,Y=0,Z=0,W=1),Translation=(X=12.201,Y=0,Z=0),Scale3D=(X=1,Y=1 ...(truncated)"
    WeightMinimum=0
    Weight=@RigVMFunction_AnimEvalRichCurve.Result
   )
   > &RigUnit_ModifyTransforms(
    ItemToModify="((Item=(Type=Bone,Name=\"base_hood_r1\"),Transform=(Rotation=(X=0,Y=0,Z=0,W=1),Translation=(X=-7.0578,Y=-14.5,Z=9.2),Scale3D=(X=1,Y=1,Z=1))),(Item=(Type=Bone,Name=\"base_hood_r\"),Transform=(Rotation=(X=0,Y=0,Z=0,W=1),Translation=(X=-14.492,Y=-17,Z=-4.3),Scale ...(truncated)"
    WeightMinimum=0
    Weight=@DeltaRotation.Delta_X_Positive
   )
   > &RigUnit_ModifyTransforms(
    ItemToModify=((Item=(Type=Bone,Name="None"),Transform=(Rotation=(X=0,Y=0,Z=0,W=1),Translation=(X=0,Y=0,Z=0),Scale3D=(X=1,Y=1,Z=1))),(Item=(Type=Bone,Name="dyn_hood_c"),Transform=(Rotation=(X=0,Y=0,Z=0,W=1),Translation=(X=0,Y=0,Z=0),Scale3D=(X=1,Y=1,Z=1))))
    WeightMinimum=0
    Weight=@DeltaRotation.Delta_z
   )
   > DeltaRotation_1 = RigVMFunctionReferenceNode(Child=Head, Parent=neck_01, Source Max=-5, Source Min=-40)
   > &RigUnit_ModifyTransforms(
    ItemToModify="((Item=(Type=Bone,Name=\"base_hood_l1\"),Transform=(Rotation=(X=0,Y=0,Z=0,W=1),Translation=(X=-0.2555,Y=-7.3279,Z=0),Scale3D=(X=1,Y=1,Z=1))),(Item=(Type=Bone,Name=\"base_hood_l\"),Transform=(Rotation=(X=0,Y=0,Z=0,W=1),Translation=(X=1.5745,Y=0,Z=0),Scale3D=(X= ...(truncated)"
    WeightMinimum=0
    Weight=@DeltaRotation_1.Delta_z
   )
   Multiply = &RigVMFunction_MathDoubleMul(A=@DeltaRotation_1.Delta_z, B=@RigVMFunction_AnimEvalRichCurve.Result)
   RigVMRerouteNode(Value=@DeltaRotation.Delta_X_Negative)
  ]}
]
hierarchy [
 root {loc=<0 0 0>, rot=<0 0 0 1>}
 pelvis {parent=root, loc=<0 -0.7071 93.1191>, rot=<0.0127 0.707 -0.0127 -0.707>}
 spine_01 {parent=pelvis, loc=<6.7494 -0.6388 0>, rot=<0 0 0.034 -0.9994>}
 spine_02 {parent=spine_01, loc=<7.252 -0.035 0>, rot=<0 0 0.0238 -0.9997>}
 spine_03 {parent=spine_02, loc=<6.0106 0 0>, rot=<0 0 0.0149 -0.9999>}
 spine_04 {parent=spine_03, loc=<8.1023 0 0>, rot=<0 0 -0.0132 -0.9999>}
 spine_05 {parent=spine_04, loc=<10.0142 0 0>, rot=<0 0 -0.0015 -1>}
 attach_backpack {parent=spine_05, loc=<-0.123 -13.473 0>, rot=<-0.0093 -0.707 0.0093 -0.707>}
 neck_01 {parent=spine_05, loc=<11.3138 0.1529 0>, rot=<0 0 -0.2003 -0.9797>}
 neck_02 {parent=neck_01, loc=<4.5627 0 0>, rot=<0 0 0.0504 -0.9987>}
 head {parent=neck_02, loc=<5.3914 0 0>, rot=<0 0 0.115 -0.9934>}
 base_hood_l {parent=neck_02, loc=<7.2211 -8.7363 -1>, rot=<-0.0784 -0.7027 0.0784 -0.7027>}
 dyn_hood_l {parent=base_hood_l, loc=<0 0 0>, rot=<0 0 0 1>}
 base_hood_r {parent=neck_02, loc=<7.2211 -8.7363 1>, rot=<-0.0784 -0.7027 0.0784 -0.7027>}
 dyn_hood_r {parent=base_hood_r, loc=<0 0 0>, rot=<0 0 0 1>}
 base_hood_c {parent=neck_02, loc=<8.5824 -8.8851 0>, rot=<-0.0784 -0.7027 0.0784 -0.7027>}
 dyn_hood_c {parent=base_hood_c, loc=<0 0 0>, rot=<0 0 0 1>}
 base_hood_l1 {parent=neck_02, loc=<0.1283 8.5072 -6.6458>, rot=<-0.0784 -0.7027 0.0784 -0.7027>}
 dyn_hood_l1 {parent=base_hood_l1, loc=<0 0 0>, rot=<0 0 0 1>}
 base_hood_r1 {parent=neck_02, loc=<0.3723 7.8024 7.7841>, rot=<-0.0784 -0.7027 0.0784 -0.7027>}
 dyn_hood_r1 {parent=base_hood_r1, loc=<0 0 0>, rot=<0 0 0 1>}
 clavicle_l {parent=spine_05, loc=<4.3575 -0.1466 -2.3223>, rot=<0.0465 -0.7289 -0.009 -0.6829>}
 upperarm_l {parent=clavicle_l, loc=<15.5111 0 0>, rot=<0.0226 -0.3234 -0.0088 -0.9459>}
 upperarm_twist_01_l {parent=upperarm_l, loc=<0.0041 0 0>, rot=<0 0 0 -1>}
 deform_upperarm_up_base_l {parent=upperarm_twist_01_l, loc=<-3.2216 -1.1391 3.3311>, rot=<0 0 0 1>}
 deform_upperarm_up_l {parent=deform_upperarm_up_base_l, loc=<0 0 0>, rot=<0 0 0 1>}
 attach_cape {parent=spine_05, loc=<9.017 -11.714 0>, rot=<0.0616 -0.7044 -0.0616 -0.7044>}
 clavicle_r {parent=spine_05, loc=<4.3575 -0.1466 2.3223>, rot=<0.7289 0.0465 -0.6829 0.009>}
 upperarm_r {parent=clavicle_r, loc=<-15.5111 0 0>, rot=<0.0226 -0.3234 -0.0088 -0.9459>}
 upperarm_twist_01_r {parent=upperarm_r, loc=<-0.0041 0 0>, rot=<0 0 0 -1>}
 deform_upperarm_up_base_r {parent=upperarm_twist_01_r, loc=<3.2214 1.1391 -3.3308>, rot=<0 0 0 1>}
 deform_upperarm_up_r {parent=deform_upperarm_up_base_r, loc=<0 0 0>, rot=<0 0 0 -1>}
 base_string_r_01 {parent=spine_05, loc=<5.0979 13.0326 3.1778>, rot=<0.9917 -0.1279 -0.0107 -0.0025>}
 dyn_string_r_00 {parent=base_string_r_01, loc=<0 0 0>, rot=<-0.0528 0.0296 -0.019 -0.998>}
 dyn_string_r_01 {parent=dyn_string_r_00, loc=<-6.5512 0 0>, rot=<0 0 0 1>}
 base_string_l_00 {parent=spine_05, loc=<4.8559 12.9108 -4.2006>, rot=<-0.1675 -0.9773 0.1225 -0.042>}
 dyn_string_l_00 {parent=base_string_l_00, loc=<0 0 0>, rot=<0 0 0 1>}
 base_hem_b_00 {parent=spine_05, loc=<-2.2448 -12.2888 -0.2128>, rot=<0.0401 -0.9992 0 0>}
 base_hem_bl_00 {parent=base_hem_b_00, loc=<2.1548 0 5.2378>, rot=<0 0 0 -1>}
 dyn_hem_bl_00 {parent=base_hem_bl_00, loc=<0 0 0>, rot=<0 0 0 1>}
 dyn_hem_bl_01 {parent=dyn_hem_bl_00, loc=<10.113 0 0>, rot=<0 0 0 1>}
 deform_pec_aim_base_l {parent=spine_04, loc=<7.8728 12.6388 -2.4123>, rot=<0.0102 -0.4602 0.3979 -0.7936>}
 deform_pec_aim_l {parent=deform_pec_aim_base_l, loc=<0 0 0>, rot=<0 0 0 1>}
 deform_pec_base_l {parent=deform_pec_aim_l, loc=<10.766 0 0>, rot=<0 0 0 1>}
 deform_pec_l {parent=deform_pec_base_l, loc=<0.055 0 0>, rot=<0 0 0 1>}
 deform_pec_aim_base_r {parent=spine_04, loc=<7.8723 12.6388 2.4123>, rot=<-0.4602 -0.0102 0.7936 0.3979>}
 deform_pec_aim_r {parent=deform_pec_aim_base_r, loc=<0 0 0>, rot=<0 0 0 1>}
 deform_pec_base_r {parent=deform_pec_aim_r, loc=<-10.7662 -0.0006 -0.0007>, rot=<0 0 0 1>}
 deform_pec_r {parent=deform_pec_base_r, loc=<0 0 0>, rot=<0 0 0 1>}
 base_hem_f_00 {parent=spine_04, loc=<-0.6414 14.7878 -0.3767>, rot=<0.0416 -0.9991 0 0>}
 dyn_hem_f_00 {parent=base_hem_f_00, loc=<0 0 0>, rot=<0 0 0 1>}
 dyn_hem_f_01 {parent=dyn_hem_f_00, loc=<7 0 0>, rot=<0 0 0 1>}
 thigh_l {parent=pelvis, loc=<-2.5693 -0.0109 -7.7501>, rot=<0.0012 0.0572 0.0155 -0.9982>}
 base_belt_01 {parent=thigh_l, loc=<2.3191 -5.5428 -7.9894>, rot=<0.0609 -0.8309 -0.5529 -0.0164>}
 base_pouch_01 {parent=base_belt_01, loc=<0 0 0>, rot=<0 0 0 1>}
 dyn_pouch_01 {parent=base_pouch_01, loc=<0 0 0>, rot=<0 0 0 1>}
 base_belt_00 {parent=pelvis, loc=<2.981 10.3937 -8.5948>, rot=<0.051 0.191 0.9762 0.0891>}
 base_pouch_00 {parent=base_belt_00, loc=<0 0 0>, rot=<0 0 0 1>}
 dyn_pouch_00 {parent=base_pouch_00, loc=<0 0 0>, rot=<0 0 0 1>}
]
