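# Physics asset for the ToxicStorm medium male character's face accessory, setting up capsule bodies and constraints for its left and right strings.
# Each string's base is angularly locked to spine_05, anchoring a one-segment dynamic chain on the left and a two-segment chain on the right.
# The dynamic joints have swing limits and zero-degree twist limits.
# Large capsules cover spine_04 and spine_05, joined by an angularly free constraint.
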PhysicsAsset /BRCosmetics/Characters/Player/Male/Medium/Bodies/M_MED_ToxicStorm/Meshes/Parts/M_MED_ToxicStorm_FaceAcc_Physics
bodies {
 base_string_l_00 [{shapeType=Capsule, length=0.505, radius=0.505, shapeName=base_string_l_00_capsule, rotation=<rot 0 -90 0>}]
 base_string_r_01 [{
   shapeType=Capsule
   length=0.505
   radius=0.505
   shapeName=base_string_r_01_capsule
   center=<-0.9005 0.032 0.0063>
   rotation=<rot 0 -64.0907 -54.0581>
  }]
 dyn_string_l_00 [{shapeType=Capsule, length=6.3285, radius=1.0359, shapeName=dyn_string_l_00_capsule, center=<5.1421 0 0>, rotation=<rot 0 -97.7904 -87.8637>}]
 dyn_string_r_00 [{shapeType=Capsule, length=4.6546, radius=0.5, shapeName=dyn_string_r_00_capsule, center=<-3.2388 0 0>, rotation=<rot 0 -89.5041 -87.1299>}]
 dyn_string_r_01 [{
   shapeType=Capsule
   length=4.589
   radius=1.0156
   shapeName=dyn_string_r_01_capsule
   center=<-2.2856 1 0>
   rotation=<rot 0.359 -106.5104 -85.8926>
  }]
 spine_04 [{shapeType=Capsule, length=23.0691, radius=13.118, shapeName=spine_04_capsule, center=<0 0.8839 0>}]
 spine_05 [{shapeType=Capsule, length=16, radius=13, center=<-1.9978 0.8395 0>}]
}
constraints [
 {
  bone1Name=base_string_l_00
  bone2Name=spine_05
  swing1LimitDegrees=45
  swing1Motion=Locked
  swing2LimitDegrees=45
  swing2Motion=Locked
  twistLimitDegrees=45
  twistMotion=Locked
 }
 {bone1Name=dyn_string_l_00, bone2Name=base_string_l_00, swing1LimitDegrees=30, swing2LimitDegrees=45, twistLimitDegrees=0}
 {
  bone1Name=base_string_r_01
  bone2Name=spine_05
  swing1LimitDegrees=45
  swing1Motion=Locked
  swing2LimitDegrees=45
  swing2Motion=Locked
  twistLimitDegrees=45
  twistMotion=Locked
 }
 {bone1Name=dyn_string_r_00, bone2Name=base_string_r_01, swing1LimitDegrees=30, swing2LimitDegrees=20, twistLimitDegrees=0}
 {bone1Name=dyn_string_r_01, bone2Name=dyn_string_r_00, swing1LimitDegrees=30, swing2LimitDegrees=45, twistLimitDegrees=0}
 {
  bone1Name=spine_05
  bone2Name=spine_04
  swing1LimitDegrees=45
  swing1Motion=Free
  swing2LimitDegrees=45
  swing2Motion=Free
  twistLimitDegrees=45
  twistMotion=Free
 }
]
